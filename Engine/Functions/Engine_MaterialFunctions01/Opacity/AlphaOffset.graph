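# AlphaOffset is a library-exposed material function that offsets an alpha value while keeping it in the 0 to 1 range, useful for shrinking blurred alphas.
# It subtracts the scalar Offset input from the Alpha input, then divides by one minus Offset to rescale the remaining range.
# The result is clamped before output, and both inputs fall back to preview values when unconnected.
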MaterialFunction /Engine/Functions/Engine_MaterialFunctions01/Opacity/AlphaOffset
graph {
 properties {bExposeToLibrary=true, description=Applies an offset to alpha while keeping the range 0 to 1. Useful for blurred alphas that need to shrink.}
 nodes [
  n1 = FunctionInput() {bUsePreviewValueAsDefault=true, inputName=Offset, inputType=FunctionInput_Scalar, sortPriority=1}
  n2 = Subtract(A=FunctionInput() {bUsePreviewValueAsDefault=true, inputName=Alpha, inputType=FunctionInput_Scalar}, B=@n1)
  n3 = OneMinus(None=@n1)
  FunctionOutput(None=Clamp(None=Divide(A=@n2, B=@n3)))
 ]
}
