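# Material function that outputs a 0 to 1 gradient by distance from the camera, useful for fading particles before they clip the camera.
# By default it subtracts Fade Offset from PixelDepth, divides by Fade Length and clamps.
# The "For Vertex Shader" static switch instead derives depth by projecting the camera-to-world-position vector onto the view axis, so it works with WorldPositionOffset.
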MaterialFunction /Engine/Functions/Engine_MaterialFunctions01/Opacity/CameraDepthFade
graph {
 properties {
  bExposeToLibrary=true
  description=Creates a gradient of 0 near the camera to white at Fade Length. Useful for preventing particles from camera clipping.0
 }
 nodes [
  MaterialExpressionSubtract_1 = Subtract(A=CameraPositionWS(), B=WorldPosition().XYZ)
  n1 = Custom!(x=@MaterialExpressionSubtract_1) {
   code="length(x)"
   inputs [{inputName=x, input {expression=&self:MaterialExpressionSubtract_1}}]
  }
  n2 = DotProduct(
   A=Normalize(VectorInput=@MaterialExpressionSubtract_1)
   B=Transform(None=Constant3Vector() {constant=<rgba 0 0 -1 0>}) {transformSourceType=TRANSFORMSOURCE_View}
  )
  n3 = FunctionInput() {
   bUsePreviewValueAsDefault=true
   description=Offsets the fade
   inputName=Fade Offset
   inputType=FunctionInput_Scalar
   sortPriority=1
   previewValue=<24 0 0 1>
  }
  n4 = FunctionInput() {
   bUsePreviewValueAsDefault=true
   description=How many units it takes to fade from 0 to 1
   inputName=Fade Length
   inputType=FunctionInput_Scalar
   previewValue=<512 0 0 1>
  }
  n5 = Divide(A=Subtract(A=Multiply(A=@n1, B=@n2), B=@n3), B=@n4)
  n6 = Divide(A=Subtract(A=PixelDepth().R, B=@n3), B=@n4)
  n7 = FunctionInput(Preview=StaticBool()) {
   bUsePreviewValueAsDefault=true
   description="Default = False. Computes depth using CameraPos instead of PixelDepth so that it is usable with WorldPositionOffset shaders"
   inputName=For Vertex Shader
   inputType=FunctionInput_StaticBool
   sortPriority=2
   previewValue=<64 0 0 1>
  }
  FunctionOutput(None=Clamp(None=StaticSwitch(True=@n5, False=@n6, Value=@n7)))
 ]
}
