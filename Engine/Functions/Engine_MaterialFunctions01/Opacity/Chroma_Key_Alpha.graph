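# Material function that extracts an alpha mask from a chroma key (green screen) image, exposed in the Opacity library category.
# Identical custom HLSL nodes normalize image and key colors by brightness, fading dark values toward gray via LumaMask, and the length of their difference gives Raw Comparison.
# That distance is remapped between cutoffs and raised to exponents, producing Alpha and an inverted Despill Alpha with its own max cutoff and exponent.
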MaterialFunction /Engine/Functions/Engine_MaterialFunctions01/Opacity/Chroma_Key_Alpha
graph {
 properties {bExposeToLibrary=true, description=Extracts an alpha mask from a chroma key aka green screen input image., libraryCategoriesText [Opacity]}
 nodes [
  MaterialExpressionFunctionInput_26 = FunctionInput(Preview=Constant3Vector() {constant=<rgba 0.1222 0.765 0.2614 1>}) {
   bUsePreviewValueAsDefault=true
   description=The chroma color aka green screen color to match and remove from the input image color.
   inputName=Chroma Color
   sortPriority=1
   previewValue=<0.3 0.8 0 0.3>
  }
  MaterialExpressionFunctionInput_3 = FunctionInput(Preview=Constant() {r=0.2}) {bUsePreviewValueAsDefault=true, inputName=LumaMask, inputType=FunctionInput_Scalar}
  n1 = Custom!(Color=@MaterialExpressionFunctionInput_26, LumaMask=@MaterialExpressionFunctionInput_3) {
   code="    float Luma = dot(Color, 1);\r\n    float ColorMask = exp(-Luma * 2 * PI / LumaMask);\r\n    Color = lerp( Color, Luma, ColorMask);\r\n    return Color / (dot(Color, 2));\r\n"
   description=ExtractColor
   inputs [
    {inputName=Color, input {expression=&self:MaterialExpressionFunctionInput_26}}
    {
     inputName=LumaMask
     input {expression=&self:MaterialExpressionFunctionInput_3}
    }
   ]
  }
  MaterialExpressionFunctionInput_23 = FunctionInput() {bUsePreviewValueAsDefault=true, description=The input image to create a mask from., inputName=Image Color, previewValue=<1 1 1 1>}
  n2 = Custom(Color=@MaterialExpressionFunctionInput_23, LumaMask=@MaterialExpressionFunctionInput_3) {
   code="    float Luma = dot(Color, 1);\r\n    float ColorMask = exp(-Luma * 2 * PI / LumaMask);\r\n    Color = lerp( Color, Luma, ColorMask);\r\n    return Color / (dot(Color, 2));\r\n"
   description=ExtractColor
   inputs [
    {inputName=Color, input {expression=&self:MaterialExpressionFunctionInput_23}}
    {
     inputName=LumaMask
     input {expression=&self:MaterialExpressionFunctionInput_3}
    }
   ]
  }
  n3 = MaterialFunctionCall(Vector 3=Subtract(A=@n1, B=@n2)) {materialFunction=/Engine/Functions/Engine_MaterialFunctions02/Utility/VectorLength}
  FunctionOutput(None=@n3.V3 Length) {
   description=Returns the absolute value of the comparison between input color and chroma color. Useful for visualizing the mask before cutoff and power parameters.
   outputName=Raw Comparison
   sortPriority=7
  }
  n4 = FunctionInput() {
   bUsePreviewValueAsDefault=true
   description=The initial comparison mask will be clipped to black at this specified value.
   inputName=Alpha Cutoff Min
   inputType=FunctionInput_Scalar
   sortPriority=2
   previewValue=<0.5 0.8 0 0.3>
  }
  n5 = Subtract(A=@n3.V3 Length, B=@n4)
  n6 = FunctionInput() {
   bUsePreviewValueAsDefault=true
   description="The initial comparison mask will reach full saturation or white at this level. "
   inputName=Alpha Cutoff Max
   inputType=FunctionInput_Scalar
   sortPriority=3
   previewValue=<0.5 0.8 0 0.3>
  }
  n7 = Max(A=Divide(A=@n5, B=Subtract(A=@n6, B=@n4))) {constB=0}
  n8 = FunctionInput() {
   bUsePreviewValueAsDefault=true
   description=This applies an exponent to the comparison mask which affects how soft or hard it is. Larger number will cause the mask to have more contrast and the white area to shrink.
   inputName=Alpha Exponent
   inputType=FunctionInput_Scalar
   sortPriority=4
   previewValue=<0.5 0.8 0 0.3>
  }
  FunctionOutput(None=Clamp(None=Power(Base=@n7, Exp=@n8))) {description=Returns black where the chroma key color is and white where the rest of the image is., outputName=Alpha}
  n9 = FunctionInput() {
   bUsePreviewValueAsDefault=true
   description="This is the maximum level from the original alpha mask where saturation will be reached. The \"Alpha Cutoff Min\" parameter also applies to the despill output before this cutoff."
   inputName=Despill Cuttoff Max
   inputType=FunctionInput_Scalar
   sortPriority=5
   previewValue=<0.5 0.8 0 0.3>
  }
  n10 = Subtract(A=@n9, B=@n4)
  n11 = FunctionInput() {
   bUsePreviewValueAsDefault=true
   description=This parameter applies an exponent to the despill mask which allows tweaking of the despill gradient sharpness.
   inputName=Despill Exponent
   inputType=FunctionInput_Scalar
   sortPriority=6
   previewValue=<0.5 0.8 0 0.3>
  }
  n12 = Power(Base=Max(A=Divide(A=@n5, B=@n10)) {constB=0}, Exp=@n11)
  FunctionOutput(None=Clamp(None=OneMinus(None=@n12))) {
   description=Useful for desaturating greenscreen color or adding in a hint of ambient bloom or lighting from the new scene.
   outputName=Despill Alpha
   sortPriority=2
  }
 ]
}
